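# Physics asset for the invisible AbruptSudden biped variant in the CosmeticCompanions content.
# It holds a single sphere body on the root bone, raised above the origin.
# There are no constraints or other bodies, so it serves as one simple collision volume rather than a jointed ragdoll.
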PhysicsAsset /CosmeticCompanions/Assets/Biped/Invis/AbruptSudden_Biped_INVIS_PhysicsAsset
bodies {root [{shapeType=Sphere, radius=40, shapeName=root_sphere, center=<0 0 30>}]}
constraints []
