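# Physics asset for the invisible biped CrescentNail cosmetic companion, giving its skeletal mesh a simple collision shape.
# It holds a single capsule body on the pelvis bone, slightly offset and rotated.
# There are no constraints, so it defines one coarse body rather than a jointed ragdoll.
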PhysicsAsset /CosmeticCompanions/Assets/Biped/Invis/CrescentNail_Biped_Invis_Physics
bodies {pelvis [{shapeType=Capsule, length=40, radius=42, shapeName=pelvis_capsule, center=<3.298 0.3981 0.0462>, rotation=<rot 0 -79.9163 -90>}]}
constraints []
