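# EggBawk_Biped_invis_PhysicsAsset is the physics asset for the invisible biped variant of the EggBawk cosmetic companion.
# It holds a single body on the root bone, a zero-length capsule that acts as a sphere raised above the bone origin.
# It defines no constraints, so it provides one simple collision volume and no articulated ragdoll.
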PhysicsAsset /CosmeticCompanions/Assets/Biped/Invis/EggBawk_Biped_invis_PhysicsAsset
bodies {root [{shapeType=Capsule, length=0, radius=36.005, shapeName=root_capsule, center=<0 0 33.3895>, rotation=<rot 0 -90 0>}]}
constraints []
