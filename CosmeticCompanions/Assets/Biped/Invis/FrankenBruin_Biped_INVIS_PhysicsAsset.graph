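# Physics asset for the invisible FrankenBruin biped mesh in the CosmeticCompanions content.
# It holds a single body on the root bone, shaped as a very short, nearly spherical capsule raised above the bone origin.
# It defines no constraints or per-limb bodies, so it gives one coarse volume instead of an articulated ragdoll.
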PhysicsAsset /CosmeticCompanions/Assets/Biped/Invis/FrankenBruin_Biped_INVIS_PhysicsAsset
bodies {root [{shapeType=Capsule, length=0.505, radius=32, shapeName=root_capsule, center=<0 0 30>}]}
constraints []
