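# Physics asset for FrankenCheeks_Quadruped_invis, an invisible quadruped cosmetic companion mesh.
# It holds a single capsule body on the pelvis bone, rotated and slightly offset from the bone origin.
# With no other bodies and no constraints, it gives one simple collision volume rather than an articulated ragdoll.
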PhysicsAsset /CosmeticCompanions/Assets/Biped/Invis/FrankenCheeks_Quadruped_invis_PhysicsAsset
bodies {pelvis [{shapeType=Capsule, length=50, radius=30, shapeName=pelvis_capsule, center=<2.8583 0 0>, rotation=<rot 0 -90 90>}]}
constraints []
