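# Physics asset for LaughPlier_Biped_INVIS, an invisible biped variant in CosmeticCompanions.
# It holds a single sphere body on the pelvis bone.
# It has no constraints and no bodies on other bones, so the character's collision shape is one ball around the pelvis rather than a jointed ragdoll.
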PhysicsAsset /CosmeticCompanions/Assets/Biped/Invis/LaughPlier_Biped_INVIS_Physics
bodies {pelvis [{shapeType=Sphere, radius=28.9523, shapeName=pelvis_sphere}]}
constraints []
