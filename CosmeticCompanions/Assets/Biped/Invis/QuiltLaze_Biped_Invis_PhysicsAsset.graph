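# Physics asset for the invisible (Invis) QuiltLaze biped in the CosmeticCompanions content, providing a coarse collision shape.
# A single tilted, offset capsule on the pelvis stands in for the body, with no per-limb shapes.
# A second body with no bone name duplicates that capsule exactly.
# No constraints are defined, so the asset has no articulated ragdoll setup.
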PhysicsAsset /CosmeticCompanions/Assets/Biped/Invis/QuiltLaze_Biped_Invis_PhysicsAsset
bodies {
 None [{shapeType=Capsule, length=15, radius=29.6603, center=<16.859 -1.3917 0>, rotation=<rot 0 -94.0969 90>}]
 pelvis [{shapeType=Capsule, length=15, radius=29.6603, center=<16.859 -1.3917 0>, rotation=<rot 0 -94.0969 90>}]
}
constraints []
