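# RoseBruin_Biped_Physics is the physics asset for the RoseBruin biped cosmetic companion, stored in the Invis asset folder.
# It holds a single rotated capsule body on the pelvis bone, offset from the bone origin.
# It defines no constraints and no other bodies, so that one capsule is the only physics shape for the skeleton.
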PhysicsAsset /CosmeticCompanions/Assets/Biped/Invis/RoseBruin_Biped_Physics
bodies {pelvis [{shapeType=Capsule, length=15, radius=29.6603, center=<16.859 -1.3917 0>, rotation=<rot 0 -94.0969 90>}]}
constraints []
