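# Physics asset for the invisible head-and-body mesh of a short biped in the CosmeticCompanions plugin.
# It approximates the whole character with a single sphere body on the root bone, raised above the bone origin.
# It has no constraints or other bodies, so it provides one simple rigid collision volume rather than an articulated ragdoll.
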PhysicsAsset /CosmeticCompanions/Assets/Biped/Invis/ShortBiped_HeadBody_Biped_Invis_PhysicsAsset
bodies {root [{shapeType=Sphere, radius=45, shapeName=root_sphere, center=<0 0 34.8076>}]}
constraints []
