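# Physics asset for VastGeese_Biped_invis, the invisible biped in the CosmeticCompanions plugin.
# It holds a single capsule body on the root bone, rotated and offset upward, with a radius far larger than its length so it is nearly spherical.
# It defines no constraints, so the whole skeleton has one rigid collision shape and no per-limb bodies or ragdoll joints.
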PhysicsAsset /CosmeticCompanions/Assets/Biped/Invis/VastGeese_Biped_invis_PhysicsAsset
bodies {root [{shapeType=Capsule, length=0.5, radius=45, shapeName=root_capsule, center=<0.02 4.8954 27.516>, rotation=<rot 0 -90 0>}]}
constraints []
